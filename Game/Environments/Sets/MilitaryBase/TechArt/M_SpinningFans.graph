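# Translucent, two-sided lit material for spinning fans in the military base set.
# Rotates UVs at two time-driven speeds with an invert option, layering fan mask and normal samples, with multiplied masks shaping color and contrast-adjusted opacity and angle-corrected normals flattened.
# Fakes directional glow from squared sky-light dot circle normal, masked and tinted by time-of-day color.
# Also writes a constant temporal responsiveness output.
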Material /Game/Environments/Sets/MilitaryBase/TechArt/M_SpinningFans
graph {
 properties {
  bAutomaticallySetUsageInEditor=false
  blendMode=BLEND_Translucent
  translucencyLightingMode=TLM_VolumetricDirectional
  twoSided=true
  usedShadingModels=Default Lit
 }
 nodes [
  n1 = Multiply(
   A=TextureSample(UVs=NamedRerouteUsage()) {samplerSource=SSM_Wrap_WorldGroupSettings, texture=&/T_FanMask}.G
   B=TextureSample(UVs=NamedRerouteUsage()) {samplerSource=SSM_Wrap_WorldGroupSettings, texture=&T_FanMask}.G
  )
  n2 = ScalarParameter() {defaultValue=0.9886, parameterName=Base Color Flatten}
  Multiply(
   A=VectorParameter() {parameterName=Base Color, defaultValue=<rgba 0.0833 0.0773 0.0804 1>}.RGB
   B=LinearInterpolate(A=@n1, Alpha=Saturate(None=@n2))
  )
  n3 = Add(A=TextureCoordinate(), B=ComponentMask(None=Constant2Vector() {g=-0.5, r=-0.5}) {g=true, r=true}) {constB=0}
  n4 = Multiply(
   A=Time()
   B=LinearInterpolate(Alpha=ScalarParameter() {controlType=Enumeration, enumeration=/Game/Global/Materials/Enums/E_Mat_TrueFalse, parameterName=Invert Fan?}) {constA=-1}
  )
  n5 = Multiply(A=@n4, B=ScalarParameter() {defaultValue=-61, parameterName=Speed 2})
  NamedRerouteDeclaration(None=Add(A=Rotator(Coordinate=@n3, Time=@n5) {centerX=0, centerY=0}, B=Constant2Vector() {g=-0.5, r=-0.5}) {constB=0}) {name=Rotator UVs 2, nodeColor=<rgba 1 0.7031 0 1>}
  n6 = Multiply(A=@n4, B=ScalarParameter() {defaultValue=-300, parameterName=Speed 1})
  NamedRerouteDeclaration(None=Add(A=Rotator(Coordinate=@n3, Time=@n6) {centerX=0, centerY=0}, B=Constant2Vector() {g=-0.5, r=-0.5}) {constB=0}) {name=Rotator UVs, nodeColor=<rgba 0.3828 0 1 1>}
  n7 = TextureSample(UVs=NamedRerouteUsage()) {samplerSource=SSM_Wrap_WorldGroupSettings, samplerType=SAMPLERTYPE_Normal, texture=&/T_Fan_Normal}
  n8 = TextureSample(UVs=NamedRerouteUsage()) {samplerSource=SSM_Wrap_WorldGroupSettings, samplerType=SAMPLERTYPE_Normal, texture=&T_Fan_Normal}
  Reroute(None=MaterialFunctionCall(
    Normal=MaterialFunctionCall(BaseNormal=@n7.RGB, AdditionalNormal=@n8.RGB) {materialFunction=&BlendAngleCorrectedNormals}.Result
    Flatness=ScalarParameter() {defaultValue=2, parameterName=Normal Flatten}
   ) {materialFunction=&FlattenNormal}.Result)
  n9 = Saturate(None=Add(A=Multiply(A=@n1, B=@n1), B=ScalarParameter() {defaultValue=0.9025, parameterName=Emissive Mask Bias}))
  n10 = TextureSample() {
   samplerSource=SSM_Wrap_WorldGroupSettings
   samplerType=SAMPLERTYPE_Normal
   texture=&/T_CircleNormal
  }
  n11 = DotProduct(A=SkyAtmosphereLightDirection(), B=Transform(None=@n10.RGB))
  n12 = Multiply(A=@n11, B=@n11)
  n13 = ScalarParameter() {defaultValue=0.512, parameterName=Emissive TOD Colors}
  n14 = MaterialFunctionCall() {materialFunction=&Color_Time_of_Day}
  Reroute(None=Multiply(A=Multiply(A=@n9, B=@n12), B=Multiply(A=@n13, B=@n14.Result)))
  n15 = ScalarParameter() {defaultValue=-1.1728, parameterName=Opacity Contrast Min}
  n16 = ScalarParameter() {defaultValue=1.7137, parameterName=Opacity Contrast Max}
  Reroute(None=Add(A=SmoothStep(Min=@n15, Max=@n16, Value=@n1), B=ScalarParameter() {defaultValue=0.1369, parameterName=Opacity Bias}))
  TemporalResponsivenessOutput!(None=Constant())
 ]
}
